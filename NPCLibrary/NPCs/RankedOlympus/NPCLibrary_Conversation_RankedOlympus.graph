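# Conversation database for the RankedOlympus NPC, who introduces herself as Odyssey and speaks cryptically about the Island's ancient powers.
# Its single entry point leads to a Speech node with a default line and contextual greetings for first-ever meetings, specific player outfits and match conversation count.
# Priority curve rows rank first-meeting and outfit reactions.
# The Speech node branches to link nodes for basic services, library quests, Season28 phased quests and a Ch5S1 quest entry.
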ConversationDatabase /NPCLibrary/NPCs/RankedOlympus/NPCLibrary_Conversation_RankedOlympus
entry_points [{tag=Conversation.Entry.RankedOlympus, destinations [#1]}]
nodes [
 ConversationEntryPointNode_7 {
  id=#1
  class=ConversationEntryPointNode
  outputs [#2]
  properties {entryTag {tagName=Conversation.Entry.RankedOlympus}}
 }
 Speech_C_4 {
  id=#2
  class=Speech_C
  outputs [
   #3
   #4
   #5
   #6
   #7
   #8
  ]
  properties {
   nodeName=""
   generalConfig {
    defaultMessage=Someone on this Island has been chasing powers they don't understand.
    contextualMessages [
     {
      message=I can feel an ancient power growing on this Island...
      contextRequirements [{requirement {class=FortControllerRequirement_IsFirstEverConversationWithNPC}}]
      priority {curve {rowName=Tandem.Speech.Priority.FirstEverInteraction}}
     }
     {
      message="Hope! The whispers were true..."
      contextRequirements [{requirement {
         class=FortControllerRequirement_HasCID
         properties {allowedCIds [&Character_VoidRedemption_Rebel]}
        }}]
      priority {curve {rowName=Tandem.Speech.Priority.OutfitReaction}}
     }
     {
      message=You have wandered from your true path... but destiny is not yet done with you.
      contextRequirements [{requirement {
         class=FortControllerRequirement_HasCID
         properties {allowedCIds [&Character_UndergroundRebel_Fashion]}
        }}]
      priority {curve {rowName=Tandem.Speech.Priority.OutfitReaction}}
     }
     {
      message=Even diamond will shatter when its weaknesses are exploited.
      contextRequirements [{requirement {
         class=FortControllerRequirement_HasCID
         properties {allowedCIds [/BRCosmetics/Athena/Items/Cosmetics/Characters/Character_DiamondHeart_Chic]}
        }}]
      priority {curve {rowName=Tandem.Speech.Priority.OutfitReaction}}
     }
     {
      message="I will not give you the answers you seek. Walk the path you have sought, and bear its consequences."
      contextRequirements [{requirement {class=FortControllerRequirement_HasCID, properties {allowedCIds [/BRCosmetics/Athena/Items/Cosmetics/Characters/Character_MadameMoth_Posh]}}}]
      priority {curve {rowName=Tandem.Speech.Priority.OutfitReaction}}
     }
     {
      message=Someone on this Island has been chasing powers they don't understand.
      contextRequirements [{requirement {
         class=FortControllerRequirement_CheckNPCMatchConversationCount
         properties {nPCUniqueIdTag {tagName=AISpawnerData.Type.Tandem.RankedOlympus}}
        }}]
      priority {curve {rowName=None}}
     }
     {
      message=The Mosaic is the key to what is coming. It bears a warning.
      contextRequirements [{requirement {
         class=FortControllerRequirement_CheckNPCMatchConversationCount
         properties {nPCUniqueIdTag {tagName=AISpawnerData.Type.Tandem.RankedOlympus}}
        }}]
      priority {curve {rowName=None}}
     }
     {
      message="Mundane powers like The Society always believe their influence is eternal, even just before they fall."
      contextRequirements [{requirement {
         class=FortControllerRequirement_CheckNPCMatchConversationCount
         properties {nPCUniqueIdTag {tagName=AISpawnerData.Type.Tandem.RankedOlympus}}
        }}]
      priority {curve {rowName=None}}
     }
     {
      message="I am Odyssey. I am merely a traveler, following the path of ancient insight."
      contextRequirements [{requirement {
         class=FortControllerRequirement_CheckNPCMatchConversationCount
         properties {nPCUniqueIdTag {tagName=AISpawnerData.Type.Tandem.RankedOlympus}}
        }}]
      priority {curve {rowName=None}}
     }
     {
      message=No destiny is certain... but yours hums with potential.
      contextRequirements [{requirement {
         class=FortControllerRequirement_CheckNPCMatchConversationCount
         properties {nPCUniqueIdTag {tagName=AISpawnerData.Type.Tandem.RankedOlympus}}
        }}]
      priority {curve {rowName=None}}
     }
     {
      message="These \"Medallions\" are shards of mysterious power, dug out from long-gone history."
      contextRequirements [{requirement {
         class=FortControllerRequirement_CheckNPCMatchConversationCount
         properties {nPCUniqueIdTag {tagName=AISpawnerData.Type.Tandem.RankedOlympus}}
        }}]
      priority {curve {rowName=None}}
     }
    ]
   }
  }
 }
 ConversationLinkNode_5 {id=#8, class=ConversationLinkNode, properties {remoteEntryTag {tagName=Conversation.Entry.Ch5S1.D3.Q8}}}
 ConversationLinkNode_4 {
  id=#7
  class=ConversationLinkNode
  properties {remoteEntryTag {tagName=Conversation.Entry.Season28.RankedOlympus.P3.Q01}}
 }
 ConversationLinkNode_2 {
  id=#6
  class=ConversationLinkNode
  properties {remoteEntryTag {tagName=Conversation.Entry.Season28.RankedOlympus.P2.Q01}}
 }
 ConversationLinkNode_3 {
  id=#5
  class=ConversationLinkNode
  properties {remoteEntryTag {tagName=Conversation.Entry.Season28.RankedOlympus.P1.Q01}}
 }
 ConversationLinkNode_0 {id=#4, class=ConversationLinkNode, properties {remoteEntryTag {tagName=Conversation.Entry.BasicServices}}}
 ConversationLinkNode_1 {id=#3, class=ConversationLinkNode, properties {remoteEntryTag {tagName=Conversation.Entry.NPCLibrary.Quests}}}
]
